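# Material function for the JackColby pickaxe that layers music-themed animation onto input material attributes, altering emissive, world position offset, normal and base color.
# Vertex color masks a time-driven local-axis spin, with pivots read from UV1 and normals rotated to match.
# A parallax cross texture pulses at a BPM parameter.
# UV1-quantized equalizer bars blend between random Perlin noise samples over time, using a gradient color and blade-mask texture.
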MaterialFunction /BRCosmetics/Weapons/FORT_Melee/Pickaxe_JackColby/Material/MF_JackColby_Pickaxe
graph {nodes [
  n1 = NamedRerouteUsage()
  n2 = NamedRerouteUsage()
  n3 = SetMaterialAttributes!(
   MaterialAttributes=NamedRerouteUsage()
   EmissiveColor=NamedRerouteUsage()
   WorldPositionOffset=NamedRerouteUsage()
   Normal=NamedRerouteUsage()
   BaseColor=LinearInterpolate(A=Constant3Vector() {constant=<rgba 0 0 0 1>}, B=NamedRerouteUsage(), Alpha=Subtract(A=@n1, B=@n2))
  ) {attributeSetTypes [
    #1
    #2
    #3
    #4
   ]}
  FunctionOutput(None=@n3)
  n4 = GetMaterialAttributes(None=FunctionInput() {inputType=FunctionInput_MaterialAttributes}) {attributeGetTypes [#1, #3, #4]}
  NamedRerouteDeclaration(None=@n4.MaterialAttributes) {name=InMat, nodeColor=<rgba 1 0.2812 0 1>}
  NamedRerouteDeclaration(None=@n4.EmissiveColor) {name=InEmissive, nodeColor=<rgba 0 0.0391 1 1>}
  n5 = Transform(None=Constant3Vector() {constant=<rgba 0 1 0 0>}) {transformSourceType=TRANSFORMSOURCE_Local}
  n6 = Multiply(A=Time(), B=ScalarParameter() {defaultValue=0.25, parameterName=SpinningSpeed})
  n7 = Multiply(A=VertexColor(), B=@n6)
  n8 = Transform(None=NamedRerouteUsage())
  n9 = Add(A=RotateAboutAxis(NormalizedRotationAxis=@n5, RotationAngle=@n7, PivotPoint=Constant3Vector(), Position=@n8), B=@n8)
  NamedRerouteDeclaration(None=LinearInterpolate(
    A=NamedRerouteUsage()
    B=Transform(None=@n9) {transformSourceType=TRANSFORMSOURCE_World, transformType=TRANSFORM_Tangent}
    Alpha=VertexColor()
   )) {name=OutNormal, nodeColor=<rgba 0 1 0.4141 1>}
  n10 = MaterialFunctionCall(Float2=TextureCoordinate() {coordinateIndex=1}) {materialFunction=&BreakOutFloat2Components}
  n11 = MaterialFunctionCall(R=@n10.R, G=Constant(), B=@n10.R) {materialFunction=&AppendMany}
  NamedRerouteDeclaration(None=RotateAboutAxis(
    NormalizedRotationAxis=@n5
    RotationAngle=@n7
    PivotPoint=TransformPosition(None=@n11.RGB) {transformType=TRANSFORMPOSSOURCE_World}
    Position=WorldPosition() {worldPositionShaderOffset=WPT_ExcludeAllShaderOffsets}.XYZ
   )) {name=OutWPO, nodeColor=<rgba 1 0.0938 0 1>}
  NamedRerouteDeclaration(None=@n4.BaseColor) {name=InBasecolor, nodeColor=<rgba 0 1 0.7188 1>}
  n12 = TextureCoordinate()
  n13 = AppendVector(A=ScalarParameter() {parameterName=Mask Offset U}, B=ScalarParameter() {parameterName=Mask Offset V})
  n14 = ScalarParameter() {defaultValue=1, parameterName=Cross Scale}
  n15 = OneMinus(None=ScalarParameter() {parameterName=Cross BPM Intensity})
  n16 = Time()
  n17 = Divide(A=ScalarParameter() {defaultValue=135, parameterName=BPM}) {constB=60}
  n18 = ConstantBiasScale(None=Frac(None=Multiply(A=@n16, B=@n17))) {bias=-0.5, scale=2}
  n19 = Power(Base=OneMinus(None=Abs(None=@n18)), Exp=ScalarParameter() {defaultValue=1, parameterName=Power})
  n20 = LinearInterpolate(A=@n15, Alpha=@n19)
  n21 = Divide(A=MaterialFunctionCall!(Radians=NamedRerouteUsage()) {materialFunction=/Engine/Functions/Engine_MaterialFunctions01/Math/RadiansToDegrees}.Degrees) {constB=360}
  n22 = MaterialFunctionCall(
   UVs=MaterialFunctionCall(UVs=Add(A=@n12, B=@n13), Texture Scale=Multiply(A=@n14, B=@n20)) {materialFunction=&ScaleUVsByCenter}.UVs
   Rotation Center=Constant() {r=0.5}
   "Rotation Angle (0-1)"=Multiply(A=@n21) {constB=-1}
  ) {materialFunction=&CustomRotator}
  n23 = Multiply(A=ScalarParameter() {defaultValue=1, parameterName=Cross Depth}, B=LinearInterpolate(B=@n15, Alpha=@n19) {constA=1})
  n24 = Saturate(None=MaterialFunctionCall(UVs=BumpOffset(Coordinate=@n22.Rotated Values, Height=@n23), Rotation Center=Constant() {r=0.5}, "Rotation Angle (0-1)"=@n21) {materialFunction=&CustomRotator}.Rotated Values)
  n25 = NamedRerouteUsage()
  NamedRerouteDeclaration(None=Multiply(
    A=TextureSample(UVs=@n24) {texture=/BRCosmetics/Weapons/FORT_Melee/Pickaxe_JackColby/Textures/T_JackColby_Pickaxe_Cross}.RGB
    B=OneMinus(None=@n25)
   )) {name=Cross, nodeColor=<rgba 1 0.2344 0 1>}
  NamedRerouteDeclaration(None=@n6) {name=SpinSpeed, nodeColor=<rgba 0 0.5078 1 1>}
  n26 = Add(A=Time(), B=Multiply(A=NamedRerouteUsage()) {constB=0.1})
  n27 = Divide(A=ScalarParameter() {defaultValue=135, parameterName=BPM}) {constB=60}
  n28 = ConstantBiasScale(None=Frac(None=Multiply(A=@n26, B=@n27))) {bias=-0.5, scale=2}
  n29 = Power(Base=OneMinus(None=Abs(None=@n28)), Exp=ScalarParameter() {defaultValue=5, parameterName=BPM Power})
  n30 = ScalarParameter() {defaultValue=0.25, parameterName=BPM Intensity}
  n31 = ComponentMask(None=TextureCoordinate() {coordinateIndex=1}) {r=true}
  n32 = Multiply(A=@n30, B=LinearInterpolate(Alpha=Power(Base=@n31) {constExponent=0.75}) {constA=0.01})
  n33 = Multiply(A=@n30) {constB=-1}
  n34 = NamedRerouteUsage()
  NamedRerouteDeclaration(None=Multiply(A=OneMinus(None=@n29), B=LinearInterpolate(A=@n32, B=@n33, Alpha=@n34)) {constB=0.01}) {name=BPMOffset, nodeColor=<rgba 1 0 0.0469 1>}
  n35 = ComponentMask(None=TextureCoordinate() {coordinateIndex=1}) {r=true}
  n36 = ScalarParameter() {defaultValue=40, parameterName=Sound Bars Amount}
  n37 = Divide(A=Ceil(None=Multiply(A=@n35, B=@n36)), B=@n36)
  n38 = NamedRerouteUsage()
  n39 = MaterialFunctionCall!(Input=Subtract(A=Floor(None=@n38)), min=Constant(), max=Constant() {r=1}, seed=Constant() {r=3}) {materialFunction=/InterchangeAssets/Functions/MX_RandomFloat}
  n40 = TextureSample(UVs=AppendVector(A=Add(A=@n37, B=@n39.Result), B=Constant())) {desc=PreviousBeat, samplerType=SAMPLERTYPE_LinearGrayscale, texture=/BRCosmetics/Weapons/FORT_Melee/Pickaxe_JackColby/Textures/T_PerlinNoise}
  n41 = ComponentMask(None=TextureCoordinate() {coordinateIndex=1}) {r=true}
  n42 = ScalarParameter() {defaultValue=40, parameterName=Sound Bars Amount}
  n43 = Divide(A=Ceil(None=Multiply(A=@n41, B=@n42)), B=@n42)
  n44 = Multiply(A=Time(), B=Divide(A=Constant() {r=80}) {constB=60})
  n45 = MaterialFunctionCall(
   Input=Floor(None=NamedRerouteDeclaration(None=@n44) {name=RandTime, nodeColor=<rgba 0.8281 0 1 1>})
   min=Constant()
   max=Constant() {r=1}
   seed=Constant() {r=3}
  ) {materialFunction=&MX_RandomFloat}
  n46 = TextureSample(UVs=AppendVector(A=Add(A=@n43, B=@n45.Result), B=Constant())) {samplerType=SAMPLERTYPE_LinearGrayscale, texture=&T_PerlinNoise}
  n47 = NamedRerouteUsage()
  n48 = Saturate(None=Multiply(A=Frac(None=@n47)) {constB=1.2})
  n49 = MaterialFunctionCall(
   In=NamedRerouteDeclaration(None=LinearInterpolate(A=@n40.R, B=@n46.R, Alpha=@n48)) {name=RandomMask, nodeColor=<rgba 0.6172 1 0 1>}
   Contrast=Constant() {r=0.15}
  ) {materialFunction=&CheapContrast}
  NamedRerouteDeclaration(None=Saturate(None=Multiply(A=@n49.Result))) {name=RandomOffset, nodeColor=<rgba 1 0 0.4687 1>}
  n50 = TextureCoordinate() {coordinateIndex=1}
  n51 = ScalarParameter() {defaultValue=40, parameterName=Sound Bars Amount}
  n52 = Multiply(A=ComponentMask(None=@n50) {r=true}, B=@n51)
  n53 = Floor(None=@n52)
  n54 = ComponentMask(None=@n50) {g=true}
  n55 = Constant() {r=0.01}
  n56 = TextureSample(UVs=AppendVector(A=Divide(A=@n53, B=@n51), B=Subtract(A=@n54, B=@n55))) {
   constCoordinate=1
   samplerType=SAMPLERTYPE_LinearGrayscale
   texture=/BRCosmetics/Weapons/FORT_Melee/Pickaxe_JackColby/Textures/T_Pickaxe_JackColby_BladeMask
  }
  NamedRerouteDeclaration(None=@n56.R) {name=BladeMask, nodeColor=<rgba 0.1719 1 0 1>}
  NamedRerouteDeclaration(None=@n4.Normal) {name=InNormal, nodeColor=<rgba 1 0 0.2578 1>}
  n57 = Add(A=TextureCoordinate(), B=@n13)
  n58 = ScalarParameter() {defaultValue=1, parameterName=Mask Scale}
  n59 = Multiply(A=Distance(
    A=MaterialFunctionCall(UVs=@n57, Texture Scale=@n58) {materialFunction=&ScaleUVsByCenter}.UVs
    B=Constant() {r=0.5}
   )) {constB=2}
  NamedRerouteDeclaration(None=Floor(None=Saturate(None=@n59))) {name=X Mask, nodeColor=<rgba 1 0.7266 0 1>}
  n60 = Multiply(A=NamedRerouteUsage(), B=NamedRerouteUsage())
  n61 = NamedRerouteUsage()
  n62 = VectorParameter() {parameterName=Sound Bars Color Top, defaultValue=<rgba 0.3476 0 20 1>}
  n63 = VectorParameter() {parameterName=Sound Bars Color Bot, defaultValue=<rgba 0 0 0 1>}
  n64 = Power(Base=NamedRerouteUsage(), Exp=ScalarParameter() {defaultValue=0.8, parameterName=Bars Color Range})
  n65 = Ceil(None=Subtract(A=Frac(None=@n52), B=ScalarParameter() {defaultValue=0.35, parameterName=Sound Bars Thickness}))
  n66 = TextureCoordinate() {coordinateIndex=1}
  n67 = OneMinus(None=Step(Y=ComponentMask(None=@n66) {r=true}, X=Constant() {r=0.277}) {constX=0.3}) {desc=Handle mask}
  n68 = LinearInterpolate(B=ScalarParameter() {defaultValue=0.75, parameterName=Sound Bars Max Height}, Alpha=@n67) {constA=0.9}
  n69 = Multiply(A=@n68, B=LinearInterpolate(Alpha=@n67) {constA=0.5, constB=0.1}) {constB=0.5}
  n70 = NamedRerouteUsage()
  n71 = Constant() {r=0.3}
  n72 = Saturate(None=Multiply(A=Subtract(A=@n70, B=@n71) {constB=0.1}, B=Constant() {r=2.5}) {constB=1.2})
  n73 = OneMinus(None=Subtract(A=LinearInterpolate(A=@n68, B=@n69, Alpha=@n72), B=NamedRerouteUsage()))
  n74 = SmoothStep(Min=@n73, Max=Add(A=@n73, B=Constant() {r=0.15}) {constB=0.1}, Value=NamedRerouteUsage())
  n75 = Ceil(None=ComponentMask(None=TextureCoordinate() {coordinateIndex=1}) {r=true})
  n76 = Saturate(None=Abs(None=Subtract(A=@n75)))
  n77 = Saturate(None=Multiply(A=Multiply(A=@n65, B=@n74), B=OneMinus(None=@n76) {desc=UV mask}))
  NamedRerouteDeclaration(None=LinearInterpolate(
    A=Add(A=@n60, B=@n61)
    B=LinearInterpolate(A=@n62.RGB, B=@n63.RGB, Alpha=@n64)
    Alpha=NamedRerouteDeclaration(None=@n77) {name=Bars Mask, nodeColor=<rgba 1 0 0.3984 1>}
   )) {name=OutEmissive, nodeColor=<rgba 1 0.3516 0 1>}
 ]}
